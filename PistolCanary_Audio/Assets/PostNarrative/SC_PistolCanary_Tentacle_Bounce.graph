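# Sound cue for a tentacle bounce in PistolCanary, reusing Athena giant umbrella bounce recordings.
# A random node picks one of three equally weighted waves, and a root modulator shifts it to a randomized raised pitch.
# It plays in the traps sound class with trap-launch attenuation.
# A concurrency override allows one instance per owner, blocking new plays and applying a short retrigger window.
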SoundCue /PistolCanary_Audio/Assets/PostNarrative/SC_PistolCanary_Tentacle_Bounce
root=Modulator_0
nodes [
 {id=Modulator_0, children [Random_0], properties {pitchMax=1.4, pitchMin=1.3}}
 {id=Random_0, children [WavePlayer_1,WavePlayer_2,WavePlayer_3], properties {weights [1,1,1]}}
 {id=WavePlayer_1, properties {soundWaveAssetPtr=/Game/Athena/Sounds/Quests/GiantUmbrellla/Athena_GiantUmbrella_Bounce_01}}
 {id=WavePlayer_2, properties {soundWaveAssetPtr=/Game/Athena/Sounds/Quests/GiantUmbrellla/Athena_GiantUmbrella_Bounce_02}}
 {id=WavePlayer_3, properties {soundWaveAssetPtr=/Game/Athena/Sounds/Quests/GiantUmbrellla/Athena_GiantUmbrella_Bounce_03}}
]
properties {
 attenuationSettings=/Game/Sounds/Attenuation/Fort_Traps_Launch_Athena_Att
 bOverrideConcurrency=true
 maxDistance=7000
 soundClassObject=&Fort_Traps
 volumeMultiplier=1
 concurrencyOverrides {bLimitToOwner=true, maxCount=1, resolutionRule=PreventNew, retriggerTime=0.1, volumeScaleAttackTime=0.1}
}
